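; Looping gameplay cue for the MusketSlinger die emote, which rolls a six-outcome random value to drive its particle effect.
; When looping starts, RollDice seeds a random stream from the player pawn's emote random seed, so the result is deterministic per seed.
; It stores the 0 to 5 result and passes it to the first particle component's RandomInt parameter.
; On removal, the stored value resets to -1.
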(event EventOnLoopingStart(EitherVFX-preferthisversion)
  (params (MyTarget :type Actor%20Object%20Reference) (Parameters :type Gameplay%20Cue%20Parameters%20Structure%20%28by%20ref%29) (ParticleComponents :type Array%20of%20FXSystem%20Component%20Object%20References) (AudioComponents :type Array%20of%20Audio%20Component%20Object%20References))
  (CallFunction|RollDice MyTarget ParticleComponents))

(event EventOnRemoval(EitherVFX-preferthisversion)
  (params (MyTarget :type Actor%20Object%20Reference) (Parameters :type Gameplay%20Cue%20Parameters%20Structure%20%28by%20ref%29) (ParticleComponents :type Array%20of%20FXSystem%20Component%20Object%20References) (AudioComponents :type Array%20of%20Audio%20Component%20Object%20References) (BurstCameraShakeInstance :type Camera%20Shake%20Base%20Object%20Reference) (BurstDecalInstance :type Decal%20Actor%20Object%20Reference))
  (Variables|Default|SetRandom%20Int%20Replicated -1))

(fn RollDice
  (params (FXTarget :type Object%20Reference) (FXArray :type Array%20of%20FXSystem%20Component%20Object%20References))
  (if (> (Utilities|Array|Length FXArray) 0)
    (bind _asfort_player_pawn (Utilities|Casting|CastToFortPlayerPawn (Variables|GetFXTarget)))
    (bind _output_get (Variables|Default|SetRandom%20Int%20Replicated (Math|Random|RandomIntegerinRangefromStream (Math|Random|MakeRandomStream (Emote|GetEmoteRandomSeed _asfort_player_pawn)) 0 5)))
    (Effects|Components|ParticleSystem|SetIntParameter (Utilities|Array|Get(acopy) (Variables|GetFXArray)) "RandomInt" _output_get)))

(fn ConstructionScript)
